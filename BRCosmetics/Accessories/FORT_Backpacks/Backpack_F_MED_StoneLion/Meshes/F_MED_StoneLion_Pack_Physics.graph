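# Physics asset for the StoneLion backpack cosmetic, giving its pack and left and right handles simple capsule bodies for physics simulation.
# One large capsule covers the pack, and each base handle bone is angularly locked to it.
# Each dynamic handle bone is constrained to its base handle, with an 80-degree Swing2 limit and zero-degree Swing1 and twist limits, suggesting hinge-like dangle.
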PhysicsAsset /BRCosmetics/Accessories/FORT_Backpacks/Backpack_F_MED_StoneLion/Meshes/F_MED_StoneLion_Pack_Physics
bodies {
 base_handle_l [{shapeType=Capsule, length=0, radius=0, shapeName=base_handle_l_capsule}]
 base_handle_r [{shapeType=Capsule, length=5, radius=0, shapeName=base_handle_r_capsule}]
 dyn_handle_l [{shapeType=Capsule, length=0.505, radius=0.505, shapeName=dyn_handle_l_capsule, center=<1.171 0 -0.0136>, rotation=<rot 0 0 -90>}]
 dyn_handle_r [{shapeType=Capsule, length=0.505, radius=0.505, shapeName=dyn_handle_r_capsule, center=<1.171 0 -0.0136>, rotation=<rot 0 -180 -90>}]
 pack [{
   shapeType=Capsule
   length=13.8087
   radius=9.8232
   shapeName=pack_capsule
   center=<-0.0276 -7.8569 -6.5902>
   rotation=<rot 0 -89.9885 -89.7936>
  }]
}
constraints [
 {bone1Name=dyn_handle_r, bone2Name=base_handle_r, swing1LimitDegrees=0, swing2LimitDegrees=80, twistLimitDegrees=0}
 {bone1Name=dyn_handle_l, bone2Name=base_handle_l, swing1LimitDegrees=0, swing2LimitDegrees=80, twistLimitDegrees=0}
 {
  bone1Name=base_handle_r
  bone2Name=pack
  swing1LimitDegrees=45
  swing1Motion=Locked
  swing2LimitDegrees=45
  swing2Motion=Locked
  twistLimitDegrees=45
  twistMotion=Locked
 }
 {
  bone1Name=base_handle_l
  bone2Name=pack
  swing1LimitDegrees=45
  swing1Motion=Locked
  swing2LimitDegrees=45
  swing2Motion=Locked
  twistLimitDegrees=45
  twistMotion=Locked
 }
]
