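# Anim Blueprint for the StoneLion back bling, playing a reactive animation when the wearer eliminates an enemy.
# A state machine moves from a held first frame to a one-shot on the elimination input event or repeating frontend timer, then feeds RigidBody physics.
# Update logic drives gravity override and sim-space settings from movement, though solo jump overwrites the Default settings variable.
# Kill-stat shake values never reach the AnimGraph, and their math sequence is unconnected.
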AnimBlueprint /BRCosmetics/Accessories/FORT_Backpacks/Backpack_F_MED_StoneLion/Meshes/F_MED_StoneLion_Pack_AnimBP
target_skeleton=&/F_MED_StoneLion_Pack_Skeleton
graphs [
 AnimGraph {nodes [
   AnimGraphNode_StateMachine_0 = Reactive() {is=title AnimGraphNode_StateMachine}
   n1 = RigidBody!(
    ComponentPose=Local To Component(LocalPose=@AnimGraphNode_StateMachine_0)
    Alpha
    OverrideWorldGravity:Vector=Get Gravity Override()
    SimSpaceSettings:Sim Space Settings Structure=Get Sim Space Settings Output() -> (Sim Space Settings Output:Sim Space Settings Structure)
   ) -> (Pose) {
    bOverrideWorldGravity=true
    overridePhysicsAsset=&/F_MED_StoneLion_Pack_Physics
    simulationSpace=BaseBoneSpace
    alphaBoolBlend {blendOption=Linear}
    baseBoneRef {boneName=root}
    componentAppliedLinearAccClamp=<3000 3000 3000>
    componentLinearAccScale=<0.3 0.4 0.3>
    componentLinearVelScale=<0.3 0.4 0.3>
    simSpaceSettings {
     maxAngularAcceleration=1000
     maxAngularVelocity=3
     maxLinearAcceleration=700
     maxLinearVelocity=500
     velocityScaleZ=0
     worldAlpha=0.65
     externalLinearDragV=<0.2 0.2 0.2>
    }
   }
   Output Pose(Result=Component To Local(ComponentPose=@n1)) {layerGroup=None}
  ]}
 Reactive {
  entry=Idle
  path=AnimGraph.AnimGraphNode_StateMachine_0.Reactive
  ids=id
  nodes [
   AnimStateNode_0 = Idle()
   AnimStateNode_1 = Reactive() {is=title AnimStateNode}
   AnimStateTransitionNode_2 = AnimStateTransitionNode() {crossfadeDuration=0}
   AnimStateTransitionNode_3 = AnimStateTransitionNode() {bAutomaticRuleBasedOnSequencePlayerInState=true, crossfadeDuration=0}
  ]
  links [
   {
    dst=Reactive
    rule=AnimGraph.AnimGraphNode_StateMachine_0.Reactive.AnimStateTransitionNode_2.Transition
    src=Idle
    via=AnimStateTransitionNode_2
   }
   {
    dst=Idle
    rule=AnimGraph.AnimGraphNode_StateMachine_0.Reactive.AnimStateTransitionNode_3.Transition
    src=Reactive
    via=AnimStateTransitionNode_3
   }
  ]
 }
 Idle {
  path=AnimGraph.AnimGraphNode_StateMachine_0.Reactive.AnimStateNode_0.Idle
  nodes [Output Animation Pose(Result="Sequence Evaluator (by time) 'StoneLion_Reactive_PACK'"!(ExplicitTime:float=0) -> (Pose:Pose Link Structure) {
     is=title AnimGraphNode_SequenceEvaluator
     sequence=/BRCosmetics/Accessories/FORT_Backpacks/Backpack_F_MED_StoneLion/Anim/StoneLion_Reactive_PACK
    })]
 }
 Reactive {
  path=AnimGraph.AnimGraphNode_StateMachine_0.Reactive.AnimStateNode_1.Reactive
  nodes [Output Animation Pose(Result=Sequence Player 'StoneLion_Reactive_PACK'() {bLoopAnimation=false, sequence=&StoneLion_Reactive_PACK})]
 }
 Transition {
  path=AnimGraph.AnimGraphNode_StateMachine_0.Reactive.AnimStateTransitionNode_2.Transition
  nodes [Result(bCanEnterTransition=Get Enemy Player Eliminated() -> (Enemy Player Eliminated:bool))]
 }
 Transition {path=AnimGraph.AnimGraphNode_StateMachine_0.Reactive.AnimStateTransitionNode_3.Transition, nodes [Result()]}
 EventGraph {nodes [
   n1 = Event BlueprintProcessAnimInputEvent()
   > Branch(Condition=IsNamed(self=@n1.AnimInputEvent, InEventName=aie.EnemyPlayerEliminated))
   > Set Enemy Player Eliminated(Enemy Player Eliminated:bool=true) -> (Output_Get:bool)
   Event OnCharacterCustomizationLoaded()
   > n2 = Sequence()
   n3 = Set Owning Pawn(execute=@n2.then_1, Owning Pawn:Pawn Object Reference=TryGetFortPawn()) -> (Output_Get:Pawn Object Reference)
   > Is Valid(InputObject=@n3)
   > n4 = FindGender!(execute, Object:Object Reference=Get Owning Pawn() -> (Owning Pawn:Pawn Object Reference)) -> (True:Exec, False:Exec, OutGender:EFortCustomGender Enum=Invalid)
   > Set isFemale(isFemale:bool="Equal (Enum)"(A:EFortCustomGender Enum=@n4, B:EFortCustomGender Enum=Female)) -> (Output_Get:bool)
   n5 = Cast To FortPlayerPawn(execute=@n2.then_0, Object=TryGetFortPawn())
   > n6 = WatchCosmeticStat!(
    execute:Exec
    Pawn:Fort Player Pawn Object Reference=@n5
    StatTag:Fort Stat Manager Tag Structure=(TagName="StatManager.AthenaKills")
    DefaultValue:int=0
    bBroadcastInitialValue:bool=false
    InitialValueBroadcastDelay:float=-1
    ChangeValueBroadcastDelay:float=-1
    bConsumeWaitingNotifications:bool=true
   ) -> (then:Exec, OnStatChanged:Exec, StatName:Name, StatValue:int) {is=title K2Node_AsyncAction}
   Set Active_Kill(execute=@n6.OnStatChanged, Active_Kill:bool=true) -> (Output_Get:bool)
   > Delay(Duration=Get ShakeTime() -> (ShakeTime:float))
   > Set Active_Kill(Active_Kill=false)
   Event BlueprintUpdateAnimation()
   > n7 = Sequence()
   > n8 = Gravity Override State Function DEPRECATED(CharacterPartAnimInstance=Self-Reference(), GameTimeInSeconds=GetGameTimeInSeconds(), GravityOverrideParams=Get Gravity Override Params())
   > Set Gravity Override(Gravity Override=@n8)
   n9 = Branch(execute=@n7.then_1, Condition=Get bIsFalling())
   > Set Sim Space Settings Output(Sim Space Settings Output:Sim Space Settings Structure=Get Sim Space Settings Falling() -> (Sim Space Settings Falling:Sim Space Settings Structure)) -> (Output_Get:Sim Space Settings Structure)
   n10 = OR Boolean!(A=Get bIsTacticalSprinting(), B=Get bIsSprinting(), C:bool=Get bIsMoving2D()) -> (ReturnValue)
   n11 = Branch(execute=@n9.else, Condition=AND Boolean(A=Get bIsJumping(), B=NOT Boolean(A=@n10)))
   > Set Sim Space Settings Default(Sim Space Settings Default:Sim Space Settings Structure=Get Sim Space Settings Solo Jump() -> (Sim Space Settings Solo Jump:Sim Space Settings Structure)) -> (Output_Get:Sim Space Settings Structure)
   Set Sim Space Settings Output(execute=@n11.else, Sim Space Settings Output=Get Sim Space Settings Default() -> (Sim Space Settings Default:Sim Space Settings Structure))
   Event BlueprintPostEvaluateAnimation() {is=title K2Node_Event}
   > Set Enemy Player Eliminated(Enemy Player Eliminated=false)
   AnimNotify_ReactiveReset() {is=title K2Node_Event}
   > n12 = Set Enemy Player Eliminated(Enemy Player Eliminated=false)
   Branch(execute=[@n13.then, @n12.then], Condition=Get bShouldDisplayFrontendReactivity() -> (bShouldDisplayFrontendReactivity:bool))
   > Delay(Duration=10)
   > Set Enemy Player Eliminated(Enemy Player Eliminated=true)
   n13 = Event BlueprintInitializeAnimation()
   n14 = Sequence!(execute) -> (then_0, then_1, then_2:Exec, then_3:Exec)
   > Get!(execute:Exec) -> (then:Exec, else:Exec, Skel_Mesh:Skeletal Mesh Component Object Reference)
   > n15 = Branch(Condition=Get Active_Kill() -> (Active_Kill:bool))
   > "Set FInterp(Alpha, (10: 10))"("FInterp(Alpha, (10: 10))":float="Get FInterp(Alpha, (10: 10))_0"() -> ("FInterp(Alpha, (10: 10))_0":float)) -> (Output_Get:float)
   > Delay(Duration=Get ShakeTime())
   > Set Active_Kill(Active_Kill=false)
   "Set FInterp(Alpha, (10: 10))"(execute=@n15.else, "FInterp(Alpha, (10: 10))"=0)
   n16 = GetGameTimeInSeconds()
   n17 = float * float(A=@n16, B=15)
   n18 = MultiplyByPi!(Value:double="Sin (Radians)"(A=float * float(A=@n17, B=3))) -> (ReturnValue:double)
   n19 = float * float(A=@n16, B=15)
   n20 = MultiplyByPi(Value="Sin (Radians)"(A=float * float(A=@n19, B=2)))
   Set Sin_Translate(execute=@n14.then_2, Sin_Translate:Vector=MakeVector(X=float * float(A=@n18, B=0.05), Y=0, Z=float * float(A=@n20, B=0.05))) -> (Output_Get:Vector)
   n21 = Get Active_Kill()
   Set Kill_Happened(
    execute=@n14.then_3
    Kill_Happened:double=FInterpTo(Current=Get Kill_Happened() -> (Kill_Happened:double), Target="To Float (Boolean)"(InBool=@n21), DeltaTime=0, InterpSpeed=1.5)
   ) -> (Output_Get:double)
   n22 = GetGameTimeInSeconds()
   n23 = float * float(A=@n22, B=25)
   n24 = MultiplyByPi(Value="Sin (Radians)"(A=float * float(A=@n23, B=2)))
   n25 = float * float(A=@n22, B=25)
   n26 = MultiplyByPi(Value="Sin (Radians)"(A=float * float(A=@n25, B=1)))
   Set Spinning_Rotator(execute=@n14.then_1, Spinning_Rotator:Rotator=MakeRotator(Roll=float * float(A=@n26, B=0.5), Yaw=float * float(A=@n24, B=0.5))) -> (Output_Get:Rotator)
   TryGetPawnOwner()
  ]}
]
variables [
 Active_Kill {type=boolean}
 Body Mesh {type=SkeletalMeshComponent}
 Enemy Player Eliminated {type=boolean}
 "FInterp(Alpha, (10: 10))" {type=number}
 "FInterp(Alpha, (10: 10))_0" {type=number}
 Gravity Override {type=Vector}
 Gravity Override Params {type=GravityOverrideParamsStruct}
 isFemale {type=boolean}
 isLarge {type=boolean}
 Kill_Happened {type=number}
 Owning Pawn {type=Pawn, default=null}
 Owning Player Pawn {type=FortPlayerPawn}
 ShakeTime {type=number, default=2}
 Sim Space Settings Default {
  type=SimSpaceSettings
  default {
   dampingAlpha=1
   maxAngularAcceleration=10000
   maxAngularVelocity=3
   maxLinearAcceleration=700
   maxLinearVelocity=500
   velocityScaleZ=1
   worldAlpha=0.4
   externalLinearDragV=<0.2 0.2 0.2>
  }
 }
 Sim Space Settings Falling {
  type=SimSpaceSettings
  default {
   dampingAlpha=1
   maxAngularAcceleration=10000
   maxAngularVelocity=3
   maxLinearAcceleration=50
   maxLinearVelocity=500
   velocityScaleZ=1
   worldAlpha=0.3
   externalLinearDragV=<0.1 0.1 0.1>
  }
 }
 Sim Space Settings Output {
  type=SimSpaceSettings
  default {
   dampingAlpha=1
   maxAngularAcceleration=1000
   maxAngularVelocity=3
   maxLinearAcceleration=700
   maxLinearVelocity=500
   velocityScaleZ=0
   worldAlpha=0.65
   externalLinearDragV=<0.2 0.2 0.2>
  }
 }
 Sim Space Settings Output_frontEnd {
  type=SimSpaceSettings
  default {
   dampingAlpha=1
   maxAngularAcceleration=10000
   maxAngularVelocity=3
   maxLinearAcceleration=200
   maxLinearVelocity=500
   velocityScaleZ=0
   worldAlpha=0.4
   externalLinearDragV=<0.1 0.1 0.1>
  }
 }
 Sim Space Settings Solo Jump {
  type=SimSpaceSettings
  default {
   dampingAlpha=1
   maxAngularAcceleration=10000
   maxAngularVelocity=3
   maxLinearAcceleration=1900
   maxLinearVelocity=500
   velocityScaleZ=1
   worldAlpha=0.6
   externalLinearDragV=<0.2 0.2 0.2>
  }
 }
 Sin_Translate {type=Vector}
 Skel_Mesh {type=SkeletalMeshComponent}
 Spinning_Rotator {type=Rotator}
]
